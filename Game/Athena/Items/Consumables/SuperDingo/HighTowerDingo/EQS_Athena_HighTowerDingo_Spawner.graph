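# EQS query for the HighTower Dingo spawner that picks candidate locations around the SuperDingo context.
# A spiral donut generator lays rings of points, which a geometry trace then projects straight down.
# A random test filters and weights the points, then a water-body test filters and scores for points inside water, ignoring waves.
# Test auto-sorting is off, so the listed order holds.
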EnvQuery /Game/Athena/Items/Consumables/SuperDingo/HighTowerDingo/EQS_Athena_HighTowerDingo_Spawner
options [{
  generator {
   name=EnvQueryGenerator_Donut_1
   class=EnvQueryGenerator_Donut
   properties {
    bAutoSortTests=false
    bUseSpiralPattern=true
    center=/Game/Athena/Items/Consumables/SuperDingo/EQSC_Athena_SuperDingo.EQSC_Athena_SuperDingo_C
    arcAngle {dataBinding=null, dataField=FloatValue, defaultValue=360}
    arcDirection {dirMode=Rotation}
    innerRadius {dataBinding=null, dataField=FloatValue, defaultValue=256}
    numberOfRings {dataBinding=null, dataField=IntValue, defaultValue=4}
    outerRadius {dataBinding=null, dataField=FloatValue, defaultValue=2560}
    pointsPerRing {dataBinding=null, dataField=IntValue, defaultValue=4}
    projectionData {extentX=0, navigationFilter=null, postProjectionVerticalOffset=0, traceMode=None}
   }
  }
  tests [
   EnvQueryTest_Project_0 {
    class=EnvQueryTest_Project
    properties {
     scoringFactor=1
     testComment=""
     testPurpose=Filter
     projectionData {
      bOnlyBlockingHits=false
      navigationFilter=null
      postProjectionVerticalOffset=0
      projectDown=20000
      projectUp=0
      serializedChannel=ECC_GameTraceChannel11
      traceChannel=TraceTypeQuery10
      traceMode=GeometryByChannel
     }
    }
   }
   FortQueryTest_Random_0 {class=FortQueryTest_Random, properties {floatValueMax=1, scoringFactor=5, testPurpose=FilterAndScore}}
   EnvQueryTest_InsideWaterBody_0 {
    class=EnvQueryTest_InsideWaterBody
    properties {
     bDefineReferenceValue=false
     bIgnoreExclusionVolumes=false
     bIncludeWaves=false
     bSimpleWaves=true
     boolValue=true
     clampMaxType=None
     clampMinType=None
     filterType=Match
     floatValueMax=0
     floatValueMin=0
     multipleContextFilterOp=AllPass
     multipleContextScoreOp=AverageScore
     normalizationType=Absolute
     referenceValue=0
     scoreClampMax=0
     scoreClampMin=0
     scoringEquation=Constant
     scoringFactor=1
     testComment=""
     testPurpose=FilterAndScore
    }
   }
  ]
 }]
